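# Unlit translucent particle material for star effects on the SleepyTime glider.
# A selected atlas cell and channel of the SleepyTime elements texture form a mask, optionally scaled by particle alpha.
# A color parameter, optionally tinted by particle color, feeds an emissive adjustment function.
# Dynamic parameters drive a sine/cosine orbit plus Sprite-function billboarding offset, blended in by UseSpline, which defaults to off.
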Material /Game/Gadgets/Assets/VinderTech_GliderChute/Glider_SleepyTime/FX/P_SleepyTime_Glider_Star
graph {
 properties {
  bUsedWithNiagaraMeshParticles=true
  bUsedWithNiagaraSprites=true
  bUsedWithParticleSprites=true
  blendMode=BLEND_Translucent
  pixelDepthOffsetMode=PDOM_Legacy
  shadingModel=MSM_Unlit
  twoSided=true
  usedShadingModels=Unlit
 }
 nodes [
  n1 = DynamicParameter() {paramNames [Radius,SpriteSize,Speed,Param4]}
  n2 = Multiply(A=Time(), B=@n1.Speed)
  n3 = Multiply(
   A=MaterialFunctionCall(X=Sine(None=@n2), Y=Cosine(None=@n2), Z=Constant()) {materialFunction=&MakeFloat3}.Result
   B=@n1.Radius
  )
  LinearInterpolate(
   A=Constant()
   B=Add(
    A=Transform(None=@n3) {transformSourceType=TRANSFORMSOURCE_Local}
    B=MaterialFunctionCall!(Center Location, 0-1 UVs, XY scale=@n1.SpriteSize, OverridePivot, "Normals (Optional)") {materialFunction=&Sprite}.World position offset
   )
   Alpha=ScalarParameter() {parameterName=UseSpline}
  )
  n4 = TextureCoordinate()
  n5 = ScalarParameter() {defaultValue=2, group=SubUV_count, parameterName=SubImagesHorizontal}
  n6 = Subtract(A=ScalarParameter() {defaultValue=1, group=SubUV_ID, parameterName=SubImagesID_Horizontal})
  n7 = Subtract(A=@n5)
  n8 = Add(A=Divide(A=ComponentMask(None=@n4) {r=true}, B=@n5), B=Multiply(A=Divide(B=@n5), B=Clamp(None=@n6, Max=@n7)))
  n9 = ScalarParameter() {defaultValue=2, group=SubUV_count, parameterName=SubImagesVertical, sortPriority=1}
  n10 = Subtract(A=ScalarParameter() {defaultValue=2, group=SubUV_ID, parameterName=SubImagesID_Vertical, sortPriority=1})
  n11 = Subtract(A=@n9)
  n12 = Add(A=Divide(A=ComponentMask(None=@n4) {g=true}, B=@n9), B=Multiply(A=Divide(B=@n9), B=Clamp(None=@n10, Max=@n11)))
  n13 = StaticComponentMaskParameter(None=TextureSampleParameter2D(UVs=AppendVector(A=@n8, B=@n12)) {parameterName=SubImage_Texture, texture=/Game/Accessories/FORT_Backpacks/M_MED_Sleepy_Time_Backpack/FX/T_SleepyTime_Elements}.RGB) {defaultR=true, parameterName=Channel}
  n14 = ParticleColor()
  LinearInterpolate(A=@n13, B=Multiply(A=@n14.A, B=@n13), Alpha=ScalarParameter() {parameterName=UseParticleAlpha})
  n15 = VectorParameter() {parameterName=Color, defaultValue=<rgba 1 1 1 1>}
  MaterialFunctionCall(Emissive=StaticSwitchParameter(True=Multiply(A=@n14.RGB, B=@n15.RGB), False=@n15.RGB) {defaultValue=true, parameterName=UseParticleColor}) {materialFunction=&MF_FX_EmissiveAdjustments}
  ScalarParameter() {defaultValue=20}
  ScalarParameter() {parameterName=speed}
  ScalarParameter() {parameterName=Radius}
 ]
}
